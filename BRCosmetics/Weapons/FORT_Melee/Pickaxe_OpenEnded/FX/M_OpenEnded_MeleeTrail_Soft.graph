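# Translucent, unlit, two-sided Niagara ribbon material for a soft melee trail on the OpenEnded pickaxe.
# Particle color drives emissive through MF_FX_EmissiveAdjustments, and particle alpha scales opacity.
# Opacity multiplies two panned noise samples, lifted by TextureAdd, with a procedural UV mask that tapers the tail, both edges and optionally the start.
# Scalar lerps, off by default, add depth fade (High/Epic quality only) and view-angle fading of edge-on surfaces.
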Material /BRCosmetics/Weapons/FORT_Melee/Pickaxe_OpenEnded/FX/M_OpenEnded_MeleeTrail_Soft
graph {
 properties {
  bUsedWithNiagaraRibbons=true
  blendMode=BLEND_Translucent
  opacityMaskClipValue=0.1
  pixelDepthOffsetMode=PDOM_Legacy
  shadingModel=MSM_Unlit
  twoSided=true
  usedShadingModels=Unlit
 }
 nodes [
  n1 = ParticleColor()
  MaterialFunctionCall(Emissive=@n1.RGB) {materialFunction=&MF_FX_EmissiveAdjustments}
  n2 = Panner(Coordinate=TextureCoordinate() {uTiling=0.2, vTiling=0.5}) {speedX=-0.6, speedY=-0.05}
  n3 = Panner(Coordinate=TextureCoordinate() {uTiling=0.3}) {speedX=-0.3, speedY=0.05}
  n4 = Add(
   A=Multiply(
    A=TextureSampleParameter2D(UVs=@n2) {parameterName=NoiseTexture, samplerType=SAMPLERTYPE_LinearColor, texture=&HighContrastNoise}.R
    B=TextureSampleParameter2D(UVs=@n3) {parameterName=NoiseTexture, samplerType=SAMPLERTYPE_LinearColor, texture=&HighContrastNoise}.B
   )
   B=ScalarParameter() {defaultValue=0.5, parameterName=TextureAdd}
  ) {constB=0.5}
  n5 = TextureCoordinate()
  n6 = ComponentMask(None=@n5) {r=true}
  n7 = Power(Base=OneMinus(None=@n6)) {constExponent=3}
  n8 = Divide(A=@n6) {constB=0.1}
  n9 = ComponentMask(None=@n5) {g=true}
  n10 = OneMinus(None=@n9)
  n11 = Multiply(
   A=LinearInterpolate(A=@n7, B=Multiply(A=@n8, B=@n7), Alpha=ScalarParameter() {defaultValue=1, parameterName=SoftStartLerp})
   B=Divide(A=Multiply(A=@n9, B=@n10)) {constB=0.25}
  )
  n12 = Saturate(None=Multiply(A=Multiply(A=@n4, B=@n11), B=ScalarParameter() {defaultValue=0.6, parameterName=AlphaMultiplier}) {constB=0.65})
  n13 = DepthFade(Opacity=@n12, FadeDistance=ScalarParameter() {defaultValue=50, parameterName=FadeDistance}) {fadeDistanceDefault=50}
  n14 = LinearInterpolate(A=@n12, B=QualitySwitch(Default=@n12, Low=@n12, High=@n13, Medium=@n12, Epic=@n13), Alpha=ScalarParameter() {parameterName=UseDepthFadeLerp})
  n15 = DotProduct(A=CameraVectorWS(), B=VertexNormalWS())
  n16 = Saturate(None=Multiply(A=Abs(None=@n15)) {constB=3})
  Multiply(A=@n1.A, B=LinearInterpolate(A=@n14, B=Multiply(A=@n14, B=@n16), Alpha=ScalarParameter() {parameterName=UseFresnelFade}))
 ]
}
