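# Unlit, masked, two-sided toon material for the OpenEnded pickaxe melee trail, set up for Niagara ribbons.
# It builds three hard-stepped masks, noisy speedlines, a reflected-gradient edge band and dark noisy toon lines, from panned noise and UV falloffs, then sums them.
# Particle alpha feeds each threshold, so shapes shrink as it fades.
# Emissive layers time-of-day speedline and two-tone edge colors with a fixed toonline color through MF_FX_EmissiveAdjustments.
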Material /BRCosmetics/Weapons/FORT_Melee/Pickaxe_OpenEnded/FX/M_OpenEnded_MeleeTrail_Toon
graph {
 properties {
  bUsedWithNiagaraRibbons=true
  blendMode=BLEND_Masked
  opacityMaskClipValue=0.1
  pixelDepthOffsetMode=PDOM_Legacy
  shadingModel=MSM_Unlit
  twoSided=true
  usedShadingModels=Unlit
 }
 nodes [
  n1 = TextureSampleParameter2D(UVs=Panner(Coordinate=TextureCoordinate() {uTiling=0.1, vTiling=3}) {speedX=-0.5}) {parameterName=NoiseTexture, samplerType=SAMPLERTYPE_LinearColor, texture=&HighContrastNoise}
  n2 = TextureSampleParameter2D(UVs=Panner(Coordinate=TextureCoordinate() {uTiling=0.2}) {speedX=-0.3, speedY=0.1}) {parameterName=NoiseTexture, samplerType=SAMPLERTYPE_LinearColor, texture=&HighContrastNoise}
  n3 = Multiply(A=@n2.B, B=@n2.B)
  n4 = TextureCoordinate()
  n5 = ComponentMask(None=@n4) {r=true}
  n6 = OneMinus(None=@n5)
  n7 = Multiply(A=Add(A=@n5) {constB=0.1}, B=Multiply(A=@n6, B=@n6))
  n8 = Divide(A=@n7) {constB=0.2}
  n9 = ComponentMask(None=@n4) {g=true}
  n10 = Divide(A=Multiply(A=@n9, B=OneMinus(None=@n9))) {constB=0.1}
  n11 = ParticleColor()
  n12 = Multiply(A=Multiply(A=Multiply(A=@n1.G, B=@n3), B=Multiply(A=@n8, B=@n10)), B=@n11.A)
  n13 = TextureCoordinate()
  n14 = ComponentMask(None=@n13) {r=true}
  n15 = Divide(A=@n14) {constB=0.05}
  n16 = OneMinus(None=@n14)
  n17 = Multiply(A=@n16, B=@n16)
  n18 = ComponentMask(None=@n13) {g=true}
  n19 = OneMinus(None=@n18)
  n20 = Power(Base=@n19) {constExponent=16}
  n21 = Constant() {r=0.7}
  n22 = Multiply(
   A=@n11.A
   B=Multiply(
    A=Multiply(A=@n15, B=@n17)
    B=MaterialFunctionCall!(Gradient=@n20, Center Point=@n21) {materialFunction=/Game/Effects/Fort_Effects/MaterialFunctions/MF_GradientReflected}.Result
   )
  )
  n23 = Step(X=@n22) {constX=0.1, constY=0.5}
  n24 = Panner(Coordinate=TextureCoordinate() {uTiling=0.1, vTiling=4}) {speedX=-0.4}
  n25 = TextureSampleParameter2D(UVs=Panner(Coordinate=TextureCoordinate() {uTiling=0.3, vTiling=1.5}) {speedX=-0.25}) {parameterName=NoiseTexture, samplerType=SAMPLERTYPE_LinearColor, texture=&HighContrastNoise}
  n26 = Multiply(A=Add(A=@n14) {constB=0.1}, B=@n17)
  n27 = Multiply(A=@n18, B=Multiply(A=@n19, B=@n19))
  n28 = Multiply(
   A=Multiply(
    A=TextureSampleParameter2D(UVs=@n24) {parameterName=NoiseTexture, samplerType=SAMPLERTYPE_LinearColor, texture=&HighContrastNoise}.R
    B=Multiply(A=@n25.G, B=@n25.G)
   )
   B=Multiply(A=Multiply(A=@n26, B=@n27)) {constB=12}
  )
  n29 = Step(X=Multiply(A=@n11.A, B=Multiply(A=@n28) {constB=1.5})) {constY=0.08}
  Add(A=Add(A=Step(X=@n12) {constY=0.25}, B=@n23), B=@n29)
  n30 = MaterialFunctionCall(
   Color_Morning=VectorParameter() {parameterName=SpeedlineColorMorning, defaultValue=<rgba 0.08 0.0867 0.1 1>}.RGB
   Color_Day=VectorParameter() {parameterName=SpeedlineColorDay, defaultValue=<rgba 0.12 0.13 0.15 1>}.RGB
   Color_Evening=VectorParameter() {parameterName=SpeedlineColorEvening, defaultValue=<rgba 0.08 0.0867 0.1 1>}.RGB
   Color_Night=VectorParameter() {parameterName=SpeedlineColorNight, defaultValue=<rgba 0.021 0.0307 0.06 1>}.RGB
  ) {materialFunction=&Color_Time_of_Day}
  n31 = VectorParameter() {parameterName=ToonlineColor, defaultValue=<rgba 0.01 0.01 0.01 1>}
  n32 = MaterialFunctionCall(
   Color_Morning=VectorParameter() {parameterName=EdgeColorMorning, defaultValue=<rgba 0.17 0.185 0.2 1>}.RGB
   Color_Day=VectorParameter() {parameterName=EdgeColorDay, defaultValue=<rgba 0.255 0.2775 0.3 1>}.RGB
   Color_Evening=VectorParameter() {parameterName=EdgeColorEvening, defaultValue=<rgba 0.17 0.185 0.2 1>}.RGB
   Color_Night=VectorParameter() {parameterName=EdgeColorNight, defaultValue=<rgba 0.024 0.033 0.06 1>}.RGB
  ) {materialFunction=&Color_Time_of_Day}
  n33 = Multiply(A=@n32.Result) {constB=0.6}
  n34 = Step(X=@n22) {constX=0.1, constY=0.9}
  MaterialFunctionCall(Emissive=LinearInterpolate(A=LinearInterpolate(A=@n30.Result, B=@n31.RGB, Alpha=@n29), B=LinearInterpolate(A=@n33, B=@n32.Result, Alpha=@n34), Alpha=@n23)) {materialFunction=&MF_FX_EmissiveAdjustments}
  Multiply(A=@n7) {constB=12}
 ]
}
